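# Looping mono MetaSound source for computer ambience on a CrewCut table prop, played quietly with close-range detail attenuation under the ambient props sound class.
# On Play, a random get with no-repeat memory picks a wave from an array variable and feeds a wave player.
# The player's Nearly Finished trigger returns through a deferred trigger variable to pick the next wave, chaining variations endlessly.
# A concurrency override caps simultaneous instances.
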MetaSoundSource /Hermes/Props/CrewCut/CrewCut_Table_A/Audio/MSS_ComputerB_Amb_Loop
inputs [UE.Source.OnPlay {type=Trigger}]
interfaces [UE.OutputFormat.Mono:1.0, UE.Source:1.1]
outputs ["UE.OutputFormat.Mono.Audio:0" {type=Audio}]
pages [Default {
  input_defaults {UE.Source.OnPlay=null}
  nodes [
   n1 = VariableDeferredAccessor.Trigger(Variable=InitVariable.Trigger().Variable)
   TriggerAnyTriggerAny2_0 = "TriggerAny.Trigger Any (2)"(In 0=^UE.Source.OnPlay, In 1=@n1.Value)
   ArrayRandomGet_0 = Array.Random Get:WaveAsset:Array(Next=@TriggerAnyTriggerAny2_0.Out, In Array=VariableAccessor.WaveAsset:Array(Variable=InitVariable.WaveAsset:Array().Variable).Value) {also {defaults {No Repeats=3}}}
   WavePlayer_0 = UE.Wave Player:Mono(Play=@ArrayRandomGet_0.On Next, Wave Asset=@ArrayRandomGet_0.Value)
   VariableMutator.Trigger(Variable=@n1.Variable, Value=@WavePlayer_0.On Nearly Finished)
  ]
  links [{dst="@UE.OutputFormat.Mono.Audio:0", src=WavePlayer_0.Out Mono}]
 }]
properties {
 assetImportData=&self:AssetImportData
 attenuationSettings=&/ATT_Detail_Amb_Loop_Close
 bLooping=true
 duration=10000
 maxDistance=900
 sampleRate=48000
 soundClassObject=&Fort_Ambient_Props_Athena
 volume=0.2
 concurrencyOverrides {maxCount=3}
}
