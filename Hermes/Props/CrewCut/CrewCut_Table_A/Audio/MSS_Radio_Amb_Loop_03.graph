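# Mono MetaSound source for a looping radio ambience, filed under the CrewCut Table A prop.
# On play, a single Wave Player sends the TidalCrane radio loop 03 wave straight to the mono output.
# The source is marked looping, but the player's Loop default holds a stray builder-result string instead of a boolean.
# It uses close detail-ambience attenuation, the Athena ambient props sound class and a small concurrency cap.
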MetaSoundSource /Hermes/Props/CrewCut/CrewCut_Table_A/Audio/MSS_Radio_Amb_Loop_03
inputs [UE.Source.OnPlay {type=Trigger}]
interfaces [UE.OutputFormat.Mono:1.0, UE.Source:1.1]
outputs ["UE.OutputFormat.Mono.Audio:0" {type=Audio}]
pages [Default {
  input_defaults {UE.Source.OnPlay=null}
  nodes [WavePlayer_0 = UE.Wave Player:Mono(Play=^UE.Source.OnPlay) {also {defaults {Loop="<MetaSoundBuilderResult.SUCCEEDED: 0>", Wave Asset=&/TidalCrane_Radio_Amb_Loop_03}}}]
  links [{dst="@UE.OutputFormat.Mono.Audio:0", src=WavePlayer_0.Out Mono}]
 }]
properties {
 assetImportData=&self:AssetImportData
 attenuationSettings=&/ATT_Detail_Amb_Loop_Close
 bLooping=true
 duration=10000
 maxDistance=900
 sampleRate=48000
 soundClassObject=&Fort_Ambient_Props_Athena
 concurrencyOverrides {maxCount=3}
}
