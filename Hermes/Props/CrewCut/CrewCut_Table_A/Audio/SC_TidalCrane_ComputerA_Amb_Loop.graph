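# Sound cue for the looping ambient hum of the TidalCrane ComputerA, stored with the CrewCut Table A prop audio.
# A Looping node wraps an equally weighted Random node that picks among four loop wave variations, choosing again on each repeat.
# Playback uses a close-range detail ambience attenuation asset and overrides concurrency to allow only a couple of simultaneous instances.
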SoundCue /Hermes/Props/CrewCut/CrewCut_Table_A/Audio/SC_TidalCrane_ComputerA_Amb_Loop
root=Looping_0
nodes [
 {id=Looping_0, children [Random_0]}
 {id=Random_0, children [WavePlayer_0,WavePlayer_1,WavePlayer_2,WavePlayer_3], properties {weights [1,1,1,1]}}
 {id=WavePlayer_0, properties {soundWaveAssetPtr=&/TidalCrane_ComputerA_Loop_01}}
 {id=WavePlayer_1, properties {soundWaveAssetPtr=&/TidalCrane_ComputerA_Loop_02}}
 {id=WavePlayer_2, properties {soundWaveAssetPtr=&/TidalCrane_ComputerA_Loop_03}}
 {id=WavePlayer_3, properties {soundWaveAssetPtr=&/TidalCrane_ComputerA_Loop_04}}
]
properties {
 attenuationSettings=&/ATT_Detail_Amb_Loop_Close
 bOverrideConcurrency=true
 maxDistance=900
 concurrencyOverrides {maxCount=2}
}
